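# Control Rig for the KneeLens medium female face accessory that adjusts dynamic hair bones as the head, neck and shoulders move.
# Back hair is rotated and narrowed, weighted by the larger of remapped head and neck_02 rotations relative to spine_05.
# Left and right front/side hair adjustments each blend by the larger of head rotation and hair-tip proximity to that side's clavicle.
# AlphaInterp nodes feed each ModifyTransforms weight, with interpolation disabled.
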ControlRigBlueprint /BRCosmetics/Characters/Player/Female/Medium/Bodies/F_MED_KneeLens/Meshes/Parts/F_MED_KneeLens_FaceAcc_CtrlRig
graphs [
 RigVMFunctionLibrary {links [], nodes []}
 RigVMModel {nodes [
   RigUnit_BeginExecution = &RigUnit_BeginExecution()
   n1 = &RigUnit_GetRelativeTransformForItem(Child=(Type=Bone,Name="neck_02"), Parent=(Type=Bone,Name="spine_05"), bChildInitial=False, bParentInitial=False)
   n2 = &RigVMFunction_MathDoubleRemap(
    SourceMaximum=-35
    SourceMinimum=0
    TargetMaximum=1
    TargetMinimum=0
    Value=&RigVMFunction_MathQuaternionToEuler(
     RotationOrder=ZXY
     Value=&RigUnit_GetRelativeTransformForItem(Child=(Type=Bone,Name="head"), Parent=(Type=Bone,Name="spine_05"), bChildInitial=False, bParentInitial=False).RelativeTransform.Rotation
    ).Result.Z
   )
   Maximum = &RigVMFunction_MathDoubleMax(
    B=&RigVMFunction_MathDoubleRemap(
     SourceMaximum=-15
     SourceMinimum=-5
     TargetMaximum=1
     TargetMinimum=0
     Value=&RigVMFunction_MathQuaternionToEuler(RotationOrder=ZXY, Value=@n1.RelativeTransform.Rotation).Result.Y
    ).Result
    A=@n2.Result
   )
   AlphaInterp = &RigVMFunction_AlphaInterp(InterpSpeedDecreasing=10, InterpSpeedIncreasing=10, Scale=1, bClampResult=False, bInterpResult=False, Value=@Maximum.Result)
   &RigUnit_ModifyTransforms(
    ItemToModify="((Item=(Type=Bone,Name=\"dyn_L_hair_back_1\"),Transform=(Rotation=(X=0,Y=0.0872,Z=0,W=0.9962),Translation=(X=0,Y=0,Z=0),Scale3D=(X=0.9,Y=1,Z=1))),(Item=(Type=Bone,Name=\"dyn_R_hair_back_1\"),Transform=(Rotation=(X=0,Y=0.0872,Z=0,W=0.9962),Translation=(X=0,Y=0, ...(truncated)"
    WeightMinimum=0
    Weight=@AlphaInterp.Result
    ExecutePin=@RigUnit_BeginExecution.ExecutePin
   )
   n3 = &RigVMFunction_MathVectorDistance(
    B=&RigUnit_GetTransform(Item=(Type=Bone,Name="clavicle_l"), bInitial=False).Transform.Translation
    A=&RigUnit_GetTransform(Item=(Type=Bone,Name="dyn_L_hair_front_3"), bInitial=False).Transform.Translation
   )
   n4 = &RigVMFunction_AlphaInterp(
    InterpSpeedDecreasing=5
    InterpSpeedIncreasing=5
    Scale=1
    bClampResult=False
    bInterpResult=False
    Value=&RigVMFunction_MathDoubleMax(
     B=&RigVMFunction_MathDoubleRemap(SourceMaximum=16, SourceMinimum=18, TargetMaximum=1, TargetMinimum=0, Value=@n3.Result).Result
     A=@n2.Result
    ).Result
   )
   > &RigUnit_ModifyTransforms(
    ItemToModify="((Item=(Type=Bone,Name=\"dyn_L_hair_front_2\"),Transform=(Rotation=(X=0,Y=0,Z=0,W=1),Translation=(X=0,Y=0,Z=0),Scale3D=(X=0.9,Y=1,Z=1))),(Item=(Type=Bone,Name=\"dyn_L_hair_side_2\"),Transform=(Rotation=(X=0,Y=0,Z=0,W=1),Translation=(X=0,Y=0,Z=0),Scale3D=(X=0.9 ...(truncated)"
    WeightMinimum=0
    Weight=@n4.Result
   )
   n5 = &RigVMFunction_MathVectorDistance(
    A=&RigUnit_GetTransform(Item=(Type=Bone,Name="dyn_R_hair_front_3"), bInitial=False).Transform.Translation
    B=&RigUnit_GetTransform(Item=(Type=Bone,Name="clavicle_r"), bInitial=False).Transform.Translation
   )
   n6 = &RigVMFunction_AlphaInterp(
    InterpSpeedDecreasing=5
    InterpSpeedIncreasing=5
    Scale=1
    bClampResult=False
    bInterpResult=False
    Value=&RigVMFunction_MathDoubleMax(
     B=&RigVMFunction_MathDoubleRemap(SourceMaximum=16, SourceMinimum=18, TargetMaximum=1, TargetMinimum=0, Value=@n5.Result).Result
     A=@n2.Result
    ).Result
   )
   > &RigUnit_ModifyTransforms(
    ItemToModify="((Item=(Type=Bone,Name=\"dyn_R_hair_front_2\"),Transform=(Rotation=(X=0,Y=0,Z=0,W=1),Translation=(X=0,Y=0,Z=0),Scale3D=(X=0.9,Y=1,Z=1))),(Item=(Type=Bone,Name=\"dyn_R_hair_front_2\"),Transform=(Rotation=(X=0,Y=0,Z=0,W=1),Translation=(X=0,Y=0,Z=0),Scale3D=(X=0. ...(truncated)"
    WeightMinimum=0
    Weight=@n6.Result
   )
  ]}
]
hierarchy [
 root {loc=<0 0 0>, rot=<0 0 0 1>}
 pelvis {parent=root, loc=<0 -0.5301 98.0571>, rot=<0.0127 0.707 -0.0127 -0.707>}
 spine_01 {parent=pelvis, loc=<6.1208 -0.9328 0>, rot=<0 0 0.034 -0.9994>}
 spine_02 {parent=spine_01, loc=<6.5997 -0.2095 0>, rot=<0 0 0.0238 -0.9997>}
 spine_03 {parent=spine_02, loc=<6.4914 -0.1237 0>, rot=<0 0 0.0149 -0.9999>}
 spine_04 {parent=spine_03, loc=<6.9598 0.0794 0>, rot=<0 0 -0.0132 -0.9999>}
 spine_05 {parent=spine_04, loc=<6.9987 0 0>, rot=<0 0 -0.0015 -1>}
 neck_01 {parent=spine_05, loc=<11.3138 0.1529 0>, rot=<0 0 -0.2003 -0.9797>}
 neck_02 {parent=neck_01, loc=<4.5627 0 0>, rot=<0 0 0.0504 -0.9987>}
 head {parent=neck_02, loc=<5.3914 0 0>, rot=<0 0 0.115 -0.9934>}
 dyn_L_hair_front_1 {parent=head, loc=<12.2826 5.3978 -6.5024>, rot=<-0.1165 -0.0684 0.9901 -0.038>}
 dyn_L_hair_front_2 {parent=dyn_L_hair_front_1, loc=<6.1715 0 0>, rot=<0 0.1968 0.058 -0.9787>}
 dyn_L_hair_front_3 {parent=dyn_L_hair_front_2, loc=<7.0221 0.0001 0.0002>, rot=<0 0 0 1>}
 dyn_L_hair_side_1 {parent=head, loc=<13.0932 0.6746 -6.6384>, rot=<-0.1922 0.0051 0.9813 0.0094>}
 dyn_L_hair_side_2 {parent=dyn_L_hair_side_1, loc=<6.6427 0.1389 0.1581>, rot=<0 0.2584 0 -0.966>}
 dyn_L_hair_side_3 {parent=dyn_L_hair_side_2, loc=<6.2514 0 0>, rot=<0 0 0 1>}
 dyn_L_hair_back_1 {parent=head, loc=<13.2735 -6.7841 -3.8636>, rot=<-0.0801 0.5205 0.8483 -0.0561>}
 dyn_L_hair_back_2 {parent=dyn_L_hair_back_1, loc=<6.1198 0 0>, rot=<0 0.2352 0 -0.972>}
 dyn_L_hair_back_3 {parent=dyn_L_hair_back_2, loc=<6.0566 0 0>, rot=<0 0 0 1>}
 dyn_glasses_main {parent=head, loc=<3.6347 4.1804 -0.0001>, rot=<0.0029 -0.7071 -0.0029 -0.7071>}
 dyn_R_hair_front_1 {parent=head, loc=<12.2829 5.3978 6.5024>, rot=<0.0684 -0.1165 -0.038 -0.9901>}
 dyn_R_hair_front_2 {parent=dyn_R_hair_front_1, loc=<-6.1719 0.0001 -0.0001>, rot=<0 0.1968 0.058 -0.9787>}
 dyn_R_hair_front_3 {parent=dyn_R_hair_front_2, loc=<-7.0224 -0.0001 -0.0001>, rot=<0 0 0 1>}
 dyn_R_hair_side_1 {parent=head, loc=<13.0936 0.6747 6.6384>, rot=<-0.0051 -0.1922 0.0094 -0.9813>}
 dyn_R_hair_side_2 {parent=dyn_R_hair_side_1, loc=<-6.6433 -0.1389 -0.1584>, rot=<0 0.2584 0 -0.966>}
 dyn_R_hair_side_3 {parent=dyn_R_hair_side_2, loc=<-6.2513 0 0>, rot=<0 0 0 -1>}
 dyn_R_hair_back_1 {parent=head, loc=<13.2733 -6.7841 3.8636>, rot=<-0.5205 -0.0801 -0.0561 -0.8483>}
 dyn_R_hair_back_2 {parent=dyn_R_hair_back_1, loc=<-6.1192 0 0.0001>, rot=<0 0.2352 0 -0.972>}
 dyn_R_hair_back_3 {parent=dyn_R_hair_back_2, loc=<-6.0572 0 0.0002>, rot=<0 0 0 -1>}
 dyn_L_strand_1 {parent=head, loc=<11.0076 7.7858 -4.6966>, rot=<-0.0573 -0.8813 0.466 -0.053>}
 dyn_L_strand_2 {parent=dyn_L_strand_1, loc=<4.0987 -0.0294 -0.0158>, rot=<0 0 0 1>}
 dyn_R_strand_1 {parent=head, loc=<10.2959 7.5854 5.1695>, rot=<0.0269 -0.9357 0.3517 0.0103>}
 dyn_R_strand_2 {parent=dyn_R_strand_1, loc=<3.742 0.068 -0.4102>, rot=<0 0.0565 0 -0.9984>}
 faceAttach {parent=head, loc=<1.1215 -1.1201 0.4822>, rot=<0.0029 -0.7071 -0.0029 -0.7071>}
 C_brow_mid {parent=faceAttach, loc=<0.4563 10.4718 2.5838>, rot=<0.1321 0 0 -0.9912>}
 L_brow_mid {parent=faceAttach, loc=<3.5439 9.9573 3.259>, rot=<0.1292 0.093 0.1791 -0.9709>}
 L_brow_outer {parent=faceAttach, loc=<5.7038 8.072 3.449>, rot=<-0.0287 -0.04 0.4491 -0.8921>}
 R_brow_mid {parent=faceAttach, loc=<-2.5795 9.9573 3.259>, rot=<0.1292 -0.093 -0.1791 -0.9709>}
 R_brow_outer {parent=faceAttach, loc=<-4.7394 8.072 3.449>, rot=<-0.0287 0.04 -0.4491 -0.8921>}
 clavicle_l {parent=spine_05, loc=<5.4831 -0.1053 -3.2946>, rot=<0.0465 -0.7289 -0.009 -0.6829>}
 clavicle_r {parent=spine_05, loc=<5.4831 -0.1053 3.2946>, rot=<0.7289 0.0465 -0.6829 0.009>}
]
