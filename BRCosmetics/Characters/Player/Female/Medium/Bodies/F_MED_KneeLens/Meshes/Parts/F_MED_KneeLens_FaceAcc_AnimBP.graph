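# Post-process style AnimBP for the KneeLens female face accessory, adding secondary motion to left and right strand and hair bones.
# It copies the attached parent mesh's pose, runs chained per-bone Anim Dynamics with angular springs and limits, then a Control Rig.
# The update event sets gravity, front-end scale, simulation space and dynamics-enabled variables.
# Only the enabled flag visibly feeds the AnimGraph, as Control Rig alpha.
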AnimBlueprint /BRCosmetics/Characters/Player/Female/Medium/Bodies/F_MED_KneeLens/Meshes/Parts/F_MED_KneeLens_FaceAcc_AnimBP
target_skeleton=&/F_MED_KneeLens_FaceAcc_Skeleton
graphs [
 AnimGraph {nodes [
   n1 = Anim Dynamics - Bone: dyn_L_strand_1!(
    SimulationSpace:AnimPhysSimSpaceType Enum
    ComponentPose=Local To Component(LocalPose=Copy Pose From Mesh() {bUseAttachedParent=true})
    Alpha=1
   ) -> (Pose) {
    angularDampingOverride=1
    angularSpringConstant=1000
    bAngularSpring=true
    bOverrideAngularDamping=true
    bUseGravityOverride=true
    gravityScale=0.35
    simulationSpace=RootRelative
    alphaBoolBlend {blendOption=Linear}
    boundBone {boneName=dyn_L_strand_1}
    componentAppliedLinearAccClamp=<4000 4000 4000>
    componentLinearAccScale=<1 1 1>
    componentLinearVelScale=<1 1 1>
    gravityOverride {z=-980}
    physicsBodyDefinitions [{
      boundBone {boneName=dyn_L_strand_1}
      boxExtents=<17 17 17>
      constraintSetup {angularLimitsMax=<15 30 30>, angularLimitsMin=<-15 0 0>, angularTarget=<1 0 0>}
      localJointOffset=<4 0 0>
     }]
   }
   n2 = Anim Dynamics - Bone: dyn_L_strand_2!(SimulationSpace:AnimPhysSimSpaceType Enum, ComponentPose=@n1, Alpha=1) -> (Pose) {
    angularDampingOverride=1
    angularSpringConstant=1000
    bAngularSpring=true
    bOverrideAngularDamping=true
    bUseGravityOverride=true
    gravityScale=0.35
    simulationSpace=RootRelative
    alphaBoolBlend {blendOption=Linear}
    boundBone {boneName=dyn_L_strand_2}
    componentAppliedLinearAccClamp=<4000 4000 4000>
    componentLinearAccScale=<1 1 1>
    componentLinearVelScale=<1 1 1>
    gravityOverride {z=-980}
    physicsBodyDefinitions [{
      boundBone {boneName=dyn_L_strand_2}
      boxExtents=<17 17 17>
      constraintSetup {angularLimitsMax=<15 30 30>, angularLimitsMin=<-15 0 0>, angularTarget=<1 0 0>}
      localJointOffset=<4 0 0>
     }]
   }
   n3 = Anim Dynamics - Bone: dyn_R_strand_1!(SimulationSpace:AnimPhysSimSpaceType Enum, ComponentPose=@n2, Alpha=1) -> (Pose) {
    angularDampingOverride=1
    angularSpringConstant=1000
    bAngularSpring=true
    bOverrideAngularDamping=true
    bUseGravityOverride=true
    gravityScale=0.35
    simulationSpace=RootRelative
    alphaBoolBlend {blendOption=Linear}
    boundBone {boneName=dyn_R_strand_1}
    componentAppliedLinearAccClamp=<4000 4000 4000>
    componentLinearAccScale=<1 1 1>
    componentLinearVelScale=<1 1 1>
    gravityOverride {z=-980}
    physicsBodyDefinitions [{
      boundBone {boneName=dyn_R_strand_1}
      boxExtents=<15 15 15>
      constraintSetup {angularLimitsMax=<15 30 30>, angularLimitsMin=<-15 0 0>, angularTarget=<1 0 0>}
      localJointOffset=<4 0 0>
     }]
   }
   n4 = Anim Dynamics - Bone: dyn_R_strand_2!(SimulationSpace:AnimPhysSimSpaceType Enum, ComponentPose=@n3, Alpha=1) -> (Pose) {
    angularDampingOverride=1
    angularSpringConstant=1000
    bAngularSpring=true
    bOverrideAngularDamping=true
    bUseGravityOverride=true
    gravityScale=0.35
    simulationSpace=RootRelative
    alphaBoolBlend {blendOption=Linear}
    boundBone {boneName=dyn_R_strand_2}
    componentAppliedLinearAccClamp=<4000 4000 4000>
    componentLinearAccScale=<1 1 1>
    componentLinearVelScale=<1 1 1>
    gravityOverride {z=-980}
    physicsBodyDefinitions [{
      boundBone {boneName=dyn_R_strand_2}
      boxExtents=<15 15 15>
      constraintSetup {angularLimitsMax=<15 30 30>, angularLimitsMin=<-15 0 0>, angularTarget=<1 0 0>}
      localJointOffset=<4 0 0>
     }]
   }
   n5 = Anim Dynamics - Bone: dyn_L_hair_front_1!(SimulationSpace:AnimPhysSimSpaceType Enum, ComponentPose=@n4, Alpha=1, GravityOverride:Vector="0,0,-1500") -> (Pose) {
    angularDampingOverride=1
    angularSpringConstant=2000
    bAngularSpring=true
    bOverrideAngularDamping=true
    bUseGravityOverride=true
    gravityScale=0.35
    simulationSpace=RootRelative
    alphaBoolBlend {blendOption=Linear}
    alphaScaleBiasClamp {scale=0.8}
    boundBone {boneName=dyn_L_hair_front_1}
    componentAppliedLinearAccClamp=<4000 4000 4000>
    componentLinearAccScale=<0.5 0.5 0.8>
    componentLinearVelScale=<0.5 0.5 0.8>
    gravityOverride {z=-1500}
    physicsBodyDefinitions [{
      boundBone {boneName=dyn_L_hair_front_1}
      boxExtents=<18 18 18>
      constraintSetup {angularLimitsMax=<0 10 15>, angularLimitsMin=<0 0 -5>, angularTarget=<1 0 0>}
      localJointOffset=<2 0 0>
     }]
   }
   n6 = Anim Dynamics - Bone: dyn_L_hair_front_2!(SimulationSpace:AnimPhysSimSpaceType Enum, ComponentPose=@n5, Alpha=1, GravityOverride:Vector="0,0,-1500") -> (Pose) {
    angularDampingOverride=1
    angularSpringConstant=1500
    bAngularSpring=true
    bOverrideAngularDamping=true
    bUseGravityOverride=true
    gravityScale=0.35
    simulationSpace=RootRelative
    alphaBoolBlend {blendOption=Linear}
    alphaScaleBiasClamp {scale=0.8}
    boundBone {boneName=dyn_L_hair_front_2}
    componentAppliedLinearAccClamp=<4000 4000 4000>
    componentLinearAccScale=<0.5 0.5 0.8>
    componentLinearVelScale=<0.5 0.5 0.8>
    gravityOverride {z=-1500}
    physicsBodyDefinitions [{
      boundBone {boneName=dyn_L_hair_front_2}
      boxExtents=<18 18 18>
      constraintSetup {angularLimitsMax=<5 10 15>, angularLimitsMin=<-5 0 -5>, angularTarget=<1 0 0>}
      localJointOffset=<2 0 0>
     }]
   }
   n7 = Anim Dynamics - Bone: dyn_L_hair_front_3!(SimulationSpace:AnimPhysSimSpaceType Enum, ComponentPose=@n6, Alpha=1, GravityOverride:Vector="0,0,-1500") -> (Pose) {
    angularDampingOverride=1
    angularSpringConstant=1000
    bAngularSpring=true
    bOverrideAngularDamping=true
    bUseGravityOverride=true
    gravityScale=0.35
    simulationSpace=RootRelative
    alphaBoolBlend {blendOption=Linear}
    alphaScaleBiasClamp {scale=0.8}
    boundBone {boneName=dyn_L_hair_front_3}
    componentAppliedLinearAccClamp=<4000 4000 4000>
    componentLinearAccScale=<0.5 0.5 0.8>
    componentLinearVelScale=<0.5 0.5 0.8>
    gravityOverride {z=-1500}
    physicsBodyDefinitions [{
      boundBone {boneName=dyn_L_hair_front_3}
      boxExtents=<18 18 18>
      constraintSetup {angularLimitsMax=<5 10 30>, angularLimitsMin=<-5 -5 -5>, angularTarget=<1 0 0>}
      localJointOffset=<2 0 0>
     }]
   }
   n8 = Anim Dynamics - Bone: dyn_L_hair_side_1!(SimulationSpace:AnimPhysSimSpaceType Enum, ComponentPose=@n7, Alpha=1, GravityOverride:Vector="0,0,-1500") -> (Pose) {
    angularDampingOverride=1
    angularSpringConstant=2000
    bAngularSpring=true
    bOverrideAngularDamping=true
    bUseGravityOverride=true
    gravityScale=0.35
    simulationSpace=RootRelative
    alphaBoolBlend {blendOption=Linear}
    alphaScaleBiasClamp {scale=0.8}
    boundBone {boneName=dyn_L_hair_side_1}
    componentAppliedLinearAccClamp=<4000 4000 4000>
    componentLinearAccScale=<0.5 0.5 0.8>
    componentLinearVelScale=<0.5 0.5 0.8>
    gravityOverride {z=-1500}
    physicsBodyDefinitions [{
      boundBone {boneName=dyn_L_hair_side_1}
      boxExtents=<17 17 17>
      constraintSetup {angularLimitsMax=<0 10 10>, angularLimitsMin=<0 0 -10>, angularTarget=<1 0 0>}
      localJointOffset=<2 0 0>
     }]
   }
   n9 = Anim Dynamics - Bone: dyn_L_hair_side_2!(SimulationSpace:AnimPhysSimSpaceType Enum, ComponentPose=@n8, Alpha=1, GravityOverride:Vector="0,0,-1500") -> (Pose) {
    angularDampingOverride=1
    angularSpringConstant=1500
    bAngularSpring=true
    bOverrideAngularDamping=true
    bUseGravityOverride=true
    gravityScale=0.35
    simulationSpace=RootRelative
    alphaBoolBlend {blendOption=Linear}
    alphaScaleBiasClamp {scale=0.8}
    boundBone {boneName=dyn_L_hair_side_2}
    componentAppliedLinearAccClamp=<4000 4000 4000>
    componentLinearAccScale=<0.5 0.5 0.8>
    componentLinearVelScale=<0.5 0.5 0.8>
    gravityOverride {z=-1500}
    physicsBodyDefinitions [{
      boundBone {boneName=dyn_L_hair_side_2}
      boxExtents=<17 17 17>
      constraintSetup {angularLimitsMax=<5 10 20>, angularLimitsMin=<-5 0 -10>, angularTarget=<1 0 0>}
      localJointOffset=<2 0 0>
     }]
   }
   n10 = Anim Dynamics - Bone: dyn_L_hair_side_3!(SimulationSpace:AnimPhysSimSpaceType Enum, ComponentPose=@n9, Alpha=1, GravityOverride:Vector="0,0,-1500") -> (Pose) {
    angularDampingOverride=1
    angularSpringConstant=1000
    bAngularSpring=true
    bOverrideAngularDamping=true
    bUseGravityOverride=true
    gravityScale=0.35
    simulationSpace=RootRelative
    alphaBoolBlend {blendOption=Linear}
    alphaScaleBiasClamp {scale=0.8}
    boundBone {boneName=dyn_L_hair_side_3}
    componentAppliedLinearAccClamp=<4000 4000 4000>
    componentLinearAccScale=<0.5 0.5 0.8>
    componentLinearVelScale=<0.5 0.5 0.8>
    gravityOverride {z=-1500}
    physicsBodyDefinitions [{
      boundBone {boneName=dyn_L_hair_side_3}
      boxExtents=<17 17 17>
      constraintSetup {angularLimitsMax=<5 10 30>, angularLimitsMin=<-5 0 -10>, angularTarget=<1 0 0>}
      localJointOffset=<2 0 0>
     }]
   }
   n11 = Anim Dynamics - Bone: dyn_L_hair_back_1!(SimulationSpace:AnimPhysSimSpaceType Enum, ComponentPose=@n10, Alpha=1, GravityOverride:Vector="0,0,-1500") -> (Pose) {
    angularDampingOverride=1
    angularSpringConstant=6000
    bAngularSpring=true
    bOverrideAngularDamping=true
    bUseGravityOverride=true
    gravityScale=0.35
    simulationSpace=RootRelative
    alphaBoolBlend {blendOption=Linear}
    alphaScaleBiasClamp {scale=0.8}
    boundBone {boneName=dyn_L_hair_back_1}
    componentAppliedLinearAccClamp=<4000 4000 4000>
    componentLinearAccScale=<0.8 0.8 0.8>
    componentLinearVelScale=<0.8 0.8 0.8>
    gravityOverride {z=-1500}
    physicsBodyDefinitions [{
      boundBone {boneName=dyn_L_hair_back_1}
      boxExtents=<17 17 17>
      constraintSetup {angularLimitsMax=<0 10 5>, angularLimitsMin=<0 0 0>, angularTarget=<1 0 0>}
      localJointOffset=<2 0 0>
     }]
   }
   n12 = Anim Dynamics - Bone: dyn_L_hair_back_2!(SimulationSpace:AnimPhysSimSpaceType Enum, ComponentPose=@n11, Alpha=1, GravityOverride:Vector="0,0,-1500") -> (Pose) {
    angularDampingOverride=1
    angularSpringConstant=6000
    bAngularSpring=true
    bOverrideAngularDamping=true
    bUseGravityOverride=true
    gravityScale=0.35
    simulationSpace=RootRelative
    alphaBoolBlend {blendOption=Linear}
    alphaScaleBiasClamp {scale=0.8}
    boundBone {boneName=dyn_L_hair_back_2}
    componentAppliedLinearAccClamp=<4000 4000 4000>
    componentLinearAccScale=<0.8 0.8 0.8>
    componentLinearVelScale=<0.8 0.8 0.8>
    gravityOverride {z=-1500}
    physicsBodyDefinitions [{
      boundBone {boneName=dyn_L_hair_back_2}
      boxExtents=<17 17 17>
      constraintSetup {angularLimitsMax=<5 10 5>, angularLimitsMin=<-5 0 0>, angularTarget=<1 0 0>}
      localJointOffset=<2 0 0>
     }]
   }
   n13 = Anim Dynamics - Bone: dyn_L_hair_back_3!(SimulationSpace:AnimPhysSimSpaceType Enum, ComponentPose=@n12, Alpha=1, GravityOverride:Vector="0,0,-1500") -> (Pose) {
    angularDampingOverride=1
    angularSpringConstant=6000
    bAngularSpring=true
    bOverrideAngularDamping=true
    bUseGravityOverride=true
    gravityScale=0.35
    simulationSpace=RootRelative
    alphaBoolBlend {blendOption=Linear}
    alphaScaleBiasClamp {scale=0.8}
    boundBone {boneName=dyn_L_hair_back_3}
    componentAppliedLinearAccClamp=<4000 4000 4000>
    componentLinearAccScale=<0.8 0.8 0.8>
    componentLinearVelScale=<0.8 0.8 0.8>
    gravityOverride {z=-1500}
    physicsBodyDefinitions [{
      boundBone {boneName=dyn_L_hair_back_3}
      boxExtents=<17 17 17>
      constraintSetup {angularLimitsMax=<5 10 5>, angularLimitsMin=<-5 -5 0>, angularTarget=<1 0 0>}
      localJointOffset=<2 0 0>
     }]
   }
   n14 = Anim Dynamics - Bone: dyn_R_hair_front_1!(SimulationSpace:AnimPhysSimSpaceType Enum, ComponentPose=@n13, Alpha=1, GravityOverride:Vector="0,0,-1500") -> (Pose) {
    angularDampingOverride=1
    angularSpringConstant=2000
    bAngularSpring=true
    bOverrideAngularDamping=true
    bUseGravityOverride=true
    gravityScale=0.35
    simulationSpace=RootRelative
    alphaBoolBlend {blendOption=Linear}
    alphaScaleBiasClamp {scale=0.8}
    boundBone {boneName=dyn_R_hair_front_1}
    componentAppliedLinearAccClamp=<4000 4000 4000>
    componentLinearAccScale=<0.5 0.5 0.8>
    componentLinearVelScale=<0.5 0.5 0.8>
    gravityOverride {z=-1500}
    physicsBodyDefinitions [{
      boundBone {boneName=dyn_R_hair_front_1}
      boxExtents=<18 18 18>
      constraintSetup {angularLimitsMax=<0 10 15>, angularLimitsMin=<0 0 -5>, angularTarget=<1 0 0>}
      localJointOffset=<-2 0 0>
     }]
   }
   n15 = Anim Dynamics - Bone: dyn_R_hair_front_2!(SimulationSpace:AnimPhysSimSpaceType Enum, ComponentPose=@n14, Alpha=1, GravityOverride:Vector="0,0,-1500") -> (Pose) {
    angularDampingOverride=1
    angularSpringConstant=1500
    bAngularSpring=true
    bOverrideAngularDamping=true
    bUseGravityOverride=true
    gravityScale=0.35
    simulationSpace=RootRelative
    alphaBoolBlend {blendOption=Linear}
    alphaScaleBiasClamp {scale=0.8}
    boundBone {boneName=dyn_R_hair_front_2}
    componentAppliedLinearAccClamp=<4000 4000 4000>
    componentLinearAccScale=<0.5 0.5 0.8>
    componentLinearVelScale=<0.5 0.5 0.8>
    gravityOverride {z=-1500}
    physicsBodyDefinitions [{
      boundBone {boneName=dyn_R_hair_front_2}
      boxExtents=<18 18 18>
      constraintSetup {angularLimitsMax=<5 10 15>, angularLimitsMin=<-5 0 -5>, angularTarget=<1 0 0>}
      localJointOffset=<-2 0 0>
     }]
   }
   n16 = Anim Dynamics - Bone: dyn_R_hair_front_3!(SimulationSpace:AnimPhysSimSpaceType Enum, ComponentPose=@n15, Alpha=1, GravityOverride:Vector="0,0,-1500") -> (Pose) {
    angularDampingOverride=1
    angularSpringConstant=1000
    bAngularSpring=true
    bOverrideAngularDamping=true
    bUseGravityOverride=true
    gravityScale=0.35
    simulationSpace=RootRelative
    alphaBoolBlend {blendOption=Linear}
    alphaScaleBiasClamp {scale=0.8}
    boundBone {boneName=dyn_R_hair_front_3}
    componentAppliedLinearAccClamp=<4000 4000 4000>
    componentLinearAccScale=<0.5 0.5 0.8>
    componentLinearVelScale=<0.5 0.5 0.8>
    gravityOverride {z=-1500}
    physicsBodyDefinitions [{
      boundBone {boneName=dyn_R_hair_front_3}
      boxExtents=<18 18 18>
      constraintSetup {angularLimitsMax=<5 10 30>, angularLimitsMin=<-5 -5 -5>, angularTarget=<1 0 0>}
      localJointOffset=<-2 0 0>
     }]
   }
   n17 = Anim Dynamics - Bone: dyn_R_hair_side_1!(SimulationSpace:AnimPhysSimSpaceType Enum, ComponentPose=@n16, Alpha=1, GravityOverride:Vector="0,0,-1500") -> (Pose) {
    angularDampingOverride=1
    angularSpringConstant=2000
    bAngularSpring=true
    bOverrideAngularDamping=true
    bUseGravityOverride=true
    gravityScale=0.35
    simulationSpace=RootRelative
    alphaBoolBlend {blendOption=Linear}
    alphaScaleBiasClamp {scale=0.8}
    boundBone {boneName=dyn_R_hair_side_1}
    componentAppliedLinearAccClamp=<4000 4000 4000>
    componentLinearAccScale=<0.5 0.5 0.8>
    componentLinearVelScale=<0.5 0.5 0.8>
    gravityOverride {z=-1500}
    physicsBodyDefinitions [{
      boundBone {boneName=dyn_R_hair_side_1}
      boxExtents=<17 17 17>
      constraintSetup {angularLimitsMax=<0 10 10>, angularLimitsMin=<0 0 -10>, angularTarget=<1 0 0>}
      localJointOffset=<-2 0 0>
     }]
   }
   n18 = Anim Dynamics - Bone: dyn_R_hair_side_2!(SimulationSpace:AnimPhysSimSpaceType Enum, ComponentPose=@n17, Alpha=1, GravityOverride:Vector="0,0,-1500") -> (Pose) {
    angularDampingOverride=1
    angularSpringConstant=1500
    bAngularSpring=true
    bOverrideAngularDamping=true
    bUseGravityOverride=true
    gravityScale=0.35
    simulationSpace=RootRelative
    alphaBoolBlend {blendOption=Linear}
    alphaScaleBiasClamp {scale=0.8}
    boundBone {boneName=dyn_R_hair_side_2}
    componentAppliedLinearAccClamp=<4000 4000 4000>
    componentLinearAccScale=<0.5 0.5 0.8>
    componentLinearVelScale=<0.5 0.5 0.8>
    gravityOverride {z=-1500}
    physicsBodyDefinitions [{
      boundBone {boneName=dyn_R_hair_side_2}
      boxExtents=<17 17 17>
      constraintSetup {angularLimitsMax=<5 10 20>, angularLimitsMin=<-5 0 -10>, angularTarget=<1 0 0>}
      localJointOffset=<-2 0 0>
     }]
   }
   n19 = Anim Dynamics - Bone: dyn_R_hair_side_3!(SimulationSpace:AnimPhysSimSpaceType Enum, ComponentPose=@n18, Alpha=1, GravityOverride:Vector="0,0,-1500") -> (Pose) {
    angularDampingOverride=1
    angularSpringConstant=1000
    bAngularSpring=true
    bOverrideAngularDamping=true
    bUseGravityOverride=true
    gravityScale=0.35
    simulationSpace=RootRelative
    alphaBoolBlend {blendOption=Linear}
    alphaScaleBiasClamp {scale=0.8}
    boundBone {boneName=dyn_R_hair_side_3}
    componentAppliedLinearAccClamp=<4000 4000 4000>
    componentLinearAccScale=<0.5 0.5 0.8>
    componentLinearVelScale=<0.5 0.5 0.8>
    gravityOverride {z=-1500}
    physicsBodyDefinitions [{
      boundBone {boneName=dyn_R_hair_side_3}
      boxExtents=<17 17 17>
      constraintSetup {angularLimitsMax=<5 10 30>, angularLimitsMin=<-5 0 -10>, angularTarget=<1 0 0>}
      localJointOffset=<-2 0 0>
     }]
   }
   n20 = Anim Dynamics - Bone: dyn_R_hair_back_1!(SimulationSpace:AnimPhysSimSpaceType Enum, ComponentPose=@n19, Alpha=1, GravityOverride:Vector="0,0,-1500") -> (Pose) {
    angularDampingOverride=1
    angularSpringConstant=6000
    bAngularSpring=true
    bOverrideAngularDamping=true
    bUseGravityOverride=true
    gravityScale=0.35
    simulationSpace=RootRelative
    alphaBoolBlend {blendOption=Linear}
    alphaScaleBiasClamp {scale=0.8}
    boundBone {boneName=dyn_R_hair_back_1}
    componentAppliedLinearAccClamp=<4000 4000 4000>
    componentLinearAccScale=<0.8 0.8 0.8>
    componentLinearVelScale=<0.8 0.8 0.8>
    gravityOverride {z=-1500}
    physicsBodyDefinitions [{
      boundBone {boneName=dyn_R_hair_back_1}
      boxExtents=<17 17 17>
      constraintSetup {angularLimitsMax=<0 10 5>, angularLimitsMin=<0 0 0>, angularTarget=<1 0 0>}
      localJointOffset=<-2 0 0>
     }]
   }
   n21 = Anim Dynamics - Bone: dyn_R_hair_back_2!(SimulationSpace:AnimPhysSimSpaceType Enum, ComponentPose=@n20, Alpha=1, GravityOverride:Vector="0,0,-1500") -> (Pose) {
    angularDampingOverride=1
    angularSpringConstant=6000
    bAngularSpring=true
    bOverrideAngularDamping=true
    bUseGravityOverride=true
    gravityScale=0.35
    simulationSpace=RootRelative
    alphaBoolBlend {blendOption=Linear}
    alphaScaleBiasClamp {scale=0.8}
    boundBone {boneName=dyn_R_hair_back_2}
    componentAppliedLinearAccClamp=<4000 4000 4000>
    componentLinearAccScale=<0.8 0.8 0.8>
    componentLinearVelScale=<0.8 0.8 0.8>
    gravityOverride {z=-1500}
    physicsBodyDefinitions [{
      boundBone {boneName=dyn_R_hair_back_2}
      boxExtents=<17 17 17>
      constraintSetup {angularLimitsMax=<5 10 5>, angularLimitsMin=<-5 0 0>, angularTarget=<1 0 0>}
      localJointOffset=<-2 0 0>
     }]
   }
   n22 = Anim Dynamics - Bone: dyn_R_hair_back_3!(SimulationSpace:AnimPhysSimSpaceType Enum, ComponentPose=@n21, Alpha=1, GravityOverride:Vector="0,0,-1500") -> (Pose) {
    angularDampingOverride=1
    angularSpringConstant=6000
    bAngularSpring=true
    bOverrideAngularDamping=true
    bUseGravityOverride=true
    gravityScale=0.35
    simulationSpace=RootRelative
    alphaBoolBlend {blendOption=Linear}
    alphaScaleBiasClamp {scale=0.8}
    boundBone {boneName=dyn_R_hair_back_3}
    componentAppliedLinearAccClamp=<4000 4000 4000>
    componentLinearAccScale=<0.8 0.8 0.8>
    componentLinearVelScale=<0.8 0.8 0.8>
    gravityOverride {z=-1500}
    physicsBodyDefinitions [{
      boundBone {boneName=dyn_R_hair_back_3}
      boxExtents=<17 17 17>
      constraintSetup {angularLimitsMax=<5 10 5>, angularLimitsMin=<-5 -5 0>, angularTarget=<1 0 0>}
      localJointOffset=<-2 0 0>
     }]
   }
   n23 = Control Rig(Source=Component To Local(ComponentPose=@n22), Alpha=Get animDynEnabled() -> (animDynEnabled:double)) {
    alphaBoolBlend {blendOption=Linear}
    controlRigAssetReference {blueprintRigClass=&/F_MED_KneeLens_FaceAcc_CtrlRig.F_MED_KneeLens_FaceAcc_CtrlRig_C}
   }
   Output Pose(Result=@n23) {layerGroup=None}
  ]}
 EventGraph {nodes [
   Event BlueprintUpdateAnimation()
   > n1 = Gravity Override State Function DEPRECATED(CharacterPartAnimInstance=Self-Reference(), GameTimeInSeconds=Get Delta Time X(), GravityOverrideParams=Get Gravity Override Params())
   > Set Gravity Override(Gravity Override=@n1)
   > Set animDynEnabled(animDynEnabled:double="To Float (Boolean)"(InBool=Get bAnimDynamicsEnabled() -> (bAnimDynamicsEnabled:bool))) -> (Output_Get:double)
   > Set frontendAlphaScale(frontendAlphaScale:double=SelectFloat(A=0.3, B=1, bPickA=Get bIsInFrontEnd())) -> (Output_Get:double)
   > n2 = Branch(Condition=OR Boolean(A=Get bIsInFrontEnd(), B=Get bIsDBNOCarried()))
   > Set simSpaceChange(simSpaceChange:AnimPhysSimSpaceType Enum=World) -> (Output_Get:AnimPhysSimSpaceType Enum)
   Set simSpaceChange(execute=@n2.else)
  ]}
]
variables [
 animDynEnabled {type=number}
 Delta Time X {type=number}
 frontendAlphaScale {type=number}
 Gravity Override {type=Vector}
 Gravity Override Params {type=GravityOverrideParamsStruct, default {onGround_Moving {gravityOverride=<0 0 -1500>}, onGround_Standing {gravityOverride=<0 0 -1500>}}}
 simSpaceChange {type=AnimPhysSimSpaceType}
]
